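# Niagara system for the Lifeline vehicle boost cosmetic, emitting exhaust trails from the car's boost ports.
# Three active emitters draw layered ribbons, pulsing cone meshes with sprites, and spinning SubUV sprites; a second cone emitter is disabled.
# Particles are partitioned by port count onto skeletal mesh sockets, shifted by a per-car offset, with spawn groups varying their behavior.
# Painted-variant overrides set each layer's colors, while GameModeID switches ribbon and sprite lifetime and velocity.
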NiagaraSystem /VehicleCosmetics/Boosts/Boost_Lifeline/FX/NS_Boost_Lifeline
dependencies {
 usedDataInterfaces [
  &NiagaraDataInterfaceArrayColor
  &NiagaraDataInterfaceArrayFloat3
  &NiagaraDataInterfaceSpriteRendererInfo
  &NiagaraDataInterfaceCurve
  &NiagaraDataInterfaceArrayFloat
  &NiagaraDataInterfaceColorCurve
  &NiagaraDataInterfaceSkeletalMesh
 ]
 usedDynamicInputs [/Niagara/DynamicInputs/Helpers/MaskBoolBySpawnGroup]
 usedModules [
  &self:SystemSpawnScript.SystemScriptSource.SystemScriptGraph.NiagaraNodeAssignment_0.SetVariables_#1
  &SystemState
 ]
}
emitters {
 NE_LifelineRibbons {
  emitterUpdateScript [
   &EmitterState()
   &VC_FilteredSocketCount()
   &VC_RateQualityHandler(High Spawn Rate=60, Low Spawn Rate=60, Medium Spawn Rate=60, Use Custom Rate?=true)
   &VC_SpawnRibbon_v1(Number of Ribbons=$Emitter.PortCount, SpawnRate=$Emitter.Spawn Rate)
  ]
  particleSpawnScript [
   &VC_PaintedVariantsOverride() {enabled=false}
   &PartitionParticles(Number Of Partitions=$Emitter.PortCount)
   &VC_InitializeRibbon_v1(
    Color=$Particles.Color
    Custom Offset=true
    Enable Skeletal Position=true
    Flip Facing=true
    Lifetime=&MergeHelper_FloatSwitch(0=1, 1=0.6, 2=0.6, 3=?, Index=$System.GameModeID, Num=::NewEnumerator1)
    Local Vector Offset=&AddVectorToPosition(Position=<-70 0 0>, Vector=$User.DTCarOffset)
    Number of Bones Per Boneflow group=2
    Offset Amount=1
    Ribbon Width=45
   )
   &AddVelocity(Velocity=&MergeHelper_VectorSwitch(0=<-1200 0 0>, 1=<-900 0 0>, 2=<-900 0 0>, 3=?, Index=$System.GameModeID, Num=::NewEnumerator1)) {module=AddVelocity001}
  ]
  particleUpdateScript [
   &ParticleState()
   &InheritSourceMovement(
    Applied Position Scale=&Multiply_VectorByFloat(
     Vector=<1 1 1>
     Float=&FloatFromCurve(FloatCurve=DataInterface() {propertyValues {Curve {defaultValue=1, keys [{tangentMode=RCTM_User, time=0, value=0.6}, {tangentMode=RCTM_User, time=1, value=1}]}}})
    )
    ApplyForce=false
    ApplyPosition=true
    ApplyPositionFalloff=false
   )
   &DynamicMaterialParameters(
    Index 0 Param 1=&MaskFloat(Mask Source=1, Mask Parameter=$Particles.PartitionParticles.Partition, Mask=0, Masked Float=0.25, Passthrough Float=0)
    Index 0 Param 2=1
    Index 0 Param 4=2.3102
    Index 1 Param 1=&MaskFloat(Mask Source=1, Mask Parameter=$Particles.PartitionParticles.Partition, Mask=0, Masked Float=0.5, Passthrough Float=0)
    Index 1 Param 2=1
    Index 1 Param 3=&MaskFloat(Mask Source=1, Mask Parameter=$Particles.PartitionParticles.Partition, Mask=0, Masked Float=1, Passthrough Float=0)
    Index 1 Param 4=1
    Module.Index0Float0Write=true
    Module.Index0Float1Write=true
    Module.Index0Float2Write=false
    Module.Index0Float3Write=false
    Module.Index1Float0Write=true
    Module.Index1Float1Write=true
    Module.Index1Float2Write=true
    Module.Index1Float3Write=true
    Write Parameter Index 1=::NewEnumerator1
   )
   &SolveForcesAndVelocity(Clamp Velocity=false, Limit Acceleration=false)
   &VC_UpdateRibbon_v1(Enable Skeletal Position=true)
   &VC_PaintedVariantsOverride(
    Black=<rgba 0.0682 0.0682 0.075 1>
    BurntSienna=<rgba 0.4583 0.1395 0 1>
    Cobalt=<rgba 0 0.3578 1.1 1>
    Crimson=<rgba 1 0.2229 0.2836 1>
    DefaultColor=<rgba 0.5866 1.1 0.2179 1>
    ForestGreen=<rgba 0.053 0.6354 0.0824 1>
    Grey=<rgba 0.2552 0.2552 0.2552 1>
    Lime=<rgba 0.7133 0.9739 0 1>
    Orange=<rgba 1.2 0.9428 0.2004 1>
    Pink=<rgba 1 0.1979 0.8767 1>
    Purple=<rgba 0.2705 0.1638 0.7865 1>
    R1 Black:LinearColor=&ColorFromCurve(ColorCurve=DataInterface() {propertyValues {
       AlphaCurve {keys [<at 0 1 RCIM_Cubic>,<at 1 0 RCIM_Cubic>]}
       BlueCurve {keys [<at 0.1007 0.0006 RCIM_Cubic>,<at 0.4241 0.01>]}
       GreenCurve {keys [<at 0.1007 0.001 RCIM_Cubic>,<at 0.4241 0.01>]}
       RedCurve {keys [<at 0.1007 0.0001 RCIM_Cubic>,<at 0.4241 0.01>]}
      }})
    R1 BurntSienna:LinearColor=&ColorFromCurve(ColorCurve=DataInterface() {propertyValues {
       AlphaCurve {keys [<at 0 1 RCIM_Cubic>,<at 1 0 RCIM_Cubic>]}
       BlueCurve {keys [<at 0.1007 0.0063 RCIM_Cubic>,<at 0.4241 0.0011 RCIM_Cubic>]}
       GreenCurve {keys [<at 0.1007 0.0164 RCIM_Cubic>,<at 0.4241 0.0062 RCIM_Cubic>]}
       RedCurve {keys [<at 0.1007 0.042 RCIM_Cubic>,<at 0.4241 0.015 RCIM_Cubic>]}
      }})
    R1 Cobalt:LinearColor=&ColorFromCurve(ColorCurve=DataInterface() {propertyValues {
       AlphaCurve {keys [<at 0 1 RCIM_Cubic>,<at 1 0 RCIM_Cubic>]}
       BlueCurve {keys [<at 0.1007 0.5677 RCIM_Cubic>,<at 0.4241 0.162 RCIM_Cubic>]}
       GreenCurve {keys [<at 0.1007 0.0295 RCIM_Cubic>,<at 0.4241 0.0109 RCIM_Cubic>]}
       RedCurve {keys [<at 0.1007 0 RCIM_Cubic>,<at 0.4241 0.0081 RCIM_Cubic>]}
      }})
    R1 Crimson:LinearColor=&ColorFromCurve(ColorCurve=DataInterface() {propertyValues {
       AlphaCurve {keys [<at 0 1 RCIM_Cubic>,<at 1 0 RCIM_Cubic>]}
       BlueCurve {keys [<at 0.1007 0 RCIM_Cubic>,<at 0.4241 0 RCIM_Cubic>]}
       GreenCurve {keys [<at 0.1007 0.0129 RCIM_Cubic>,<at 0.4241 0.0033 RCIM_Cubic>]}
       RedCurve {keys [<at 0.1007 0.4062 RCIM_Cubic>,<at 0.4241 0.105 RCIM_Cubic>]}
      }})
    R1 DefaultColor=&ColorFromCurve(ColorCurve=DataInterface() {propertyValues {
       AlphaCurve {keys [<at 0 1 RCIM_Cubic>,<at 1 0 RCIM_Cubic>]}
       BlueCurve {keys [<at 0.1007 0.4179 RCIM_Cubic>,<at 0.4241 0.285 RCIM_Cubic>]}
       GreenCurve {keys [<at 0.1007 0.6867 RCIM_Cubic>,<at 0.4241 0.0314 RCIM_Cubic>]}
       RedCurve {keys [<at 0.1007 0.0976 RCIM_Cubic>,<at 0.4241 0 RCIM_Cubic>]}
      }})
    R1 ForestGreen:LinearColor=&ColorFromCurve(ColorCurve=DataInterface() {propertyValues {
       AlphaCurve {keys [<at 0 1 RCIM_Cubic>,<at 1 0 RCIM_Cubic>]}
       BlueCurve {keys [<at 0.1007 0.018 RCIM_Cubic>,<at 0.4241 0.01 RCIM_Cubic>]}
       GreenCurve {keys [<at 0.1007 0.17 RCIM_Cubic>,<at 0.4241 0.06 RCIM_Cubic>]}
       RedCurve {keys [<at 0.1007 0 RCIM_Cubic>,<at 0.4241 0.0004 RCIM_Cubic>]}
      }})
    R1 Grey:LinearColor=&ColorFromCurve(ColorCurve=DataInterface() {propertyValues {
       AlphaCurve {keys [<at 0 1 RCIM_Cubic>,<at 1 0 RCIM_Cubic>]}
       BlueCurve {keys [<at 0.1007 0.125 RCIM_Cubic>,<at 0.4241 0.02 RCIM_Cubic>]}
       GreenCurve {keys [<at 0.1007 0.1094 RCIM_Cubic>,<at 0.4241 0.02 RCIM_Cubic>]}
       RedCurve {keys [<at 0.1007 0.1094 RCIM_Cubic>,<at 0.4241 0.02 RCIM_Cubic>]}
      }})
    R1 Lime:LinearColor=&ColorFromCurve(ColorCurve=DataInterface() {propertyValues {
       AlphaCurve {keys [<at 0 1 RCIM_Cubic>,<at 1 0 RCIM_Cubic>]}
       BlueCurve {keys [<at 0.1007 0.008 RCIM_Cubic>,<at 0.4241 0.016 RCIM_Cubic>]}
       GreenCurve {keys [<at 0.1007 0.5104 RCIM_Cubic>,<at 0.4241 0.1458 RCIM_Cubic>]}
       RedCurve {keys [<at 0.1007 0.2124 RCIM_Cubic>,<at 0.4241 0.0535 RCIM_Cubic>]}
      }})
    R1 Orange:LinearColor=&ColorFromCurve(ColorCurve=DataInterface() {propertyValues {
       AlphaCurve {keys [<at 0 1 RCIM_Cubic>,<at 1 0 RCIM_Cubic>]}
       BlueCurve {keys [<at 0.1007 0 RCIM_Cubic>,<at 0.4241 0 RCIM_Cubic>]}
       GreenCurve {keys [<at 0.1007 0.0815 RCIM_Cubic>,<at 0.4241 0.033 RCIM_Cubic>]}
       RedCurve {keys [<at 0.1007 0.5826 RCIM_Cubic>,<at 0.4241 0.1391 RCIM_Cubic>]}
      }})
    R1 Pink:LinearColor=&ColorFromCurve(ColorCurve=DataInterface() {propertyValues {
       AlphaCurve {keys [<at 0 1 RCIM_Cubic>,<at 1 0 RCIM_Cubic>]}
       BlueCurve {keys [<at 0.1007 0.2855 RCIM_Cubic>,<at 0.4241 0.1612 RCIM_Cubic>]}
       GreenCurve {keys [<at 0.1007 0.0353 RCIM_Cubic>,<at 0.4241 0.0152 RCIM_Cubic>]}
       RedCurve {keys [<at 0.1007 0.3391 RCIM_Cubic>,<at 0.4241 0.19 RCIM_Cubic>]}
      }})
    R1 Purple:LinearColor=&ColorFromCurve(ColorCurve=DataInterface() {propertyValues {
       AlphaCurve {keys [<at 0 1 RCIM_Cubic>,<at 1 0 RCIM_Cubic>]}
       BlueCurve {keys [<at 0.1007 0.6172 RCIM_Cubic>,<at 0.4241 0.1739 RCIM_Cubic>]}
       GreenCurve {keys [<at 0.1007 0.0015 RCIM_Cubic>,<at 0.4241 0 RCIM_Cubic>]}
       RedCurve {keys [<at 0.1007 0.0544 RCIM_Cubic>,<at 0.4241 0.016 RCIM_Cubic>]}
      }})
    R1 Saffron:LinearColor=&ColorFromCurve(ColorCurve=DataInterface() {propertyValues {
       AlphaCurve {keys [<at 0 1 RCIM_Cubic>,<at 1 0 RCIM_Cubic>]}
       BlueCurve {keys [<at 0.1007 0.0289 RCIM_Cubic>,<at 0.4241 0 RCIM_Cubic>]}
       GreenCurve {keys [<at 0.1007 0.4948 RCIM_Cubic>,<at 0.4241 0.1406 RCIM_Cubic>]}
       RedCurve {keys [<at 0.1007 0.4822 RCIM_Cubic>,<at 0.4241 0.1335 RCIM_Cubic>]}
      }})
    R1 SkyBlue:LinearColor=&ColorFromCurve(ColorCurve=DataInterface() {propertyValues {
       AlphaCurve {keys [<at 0 1 RCIM_Cubic>,<at 1 0 RCIM_Cubic>]}
       BlueCurve {keys [<at 0.1007 0.6823 RCIM_Cubic>,<at 0.4241 0.1719 RCIM_Cubic>]}
       GreenCurve {keys [<at 0.1007 0.1662 RCIM_Cubic>,<at 0.4241 0.0519 RCIM_Cubic>]}
       RedCurve {keys [<at 0.1007 0 RCIM_Cubic>,<at 0.4241 0 RCIM_Cubic>]}
      }})
    R1 TitaniumWhite:LinearColor=&ColorFromCurve(ColorCurve=DataInterface() {propertyValues {
       AlphaCurve {keys [<at 0 1 RCIM_Cubic>,<at 1 0 RCIM_Cubic>]}
       BlueCurve {keys [<at 0.1007 0.528 RCIM_Cubic>,<at 0.4241 0.213 RCIM_Cubic>]}
       GreenCurve {keys [<at 0.1007 0.5221 RCIM_Cubic>,<at 0.4241 0.2031 RCIM_Cubic>]}
       RedCurve {keys [<at 0.1007 0.5221 RCIM_Cubic>,<at 0.4241 0.2031 RCIM_Cubic>]}
      }})
    R2 Black:LinearColor=<rgba 0.008 0.008 0.008 1>
    R2 BurntSienna:LinearColor=<rgba 0.0642 0.0276 0.0177 1>
    R2 Cobalt:LinearColor=<rgba 0.0677 0.085 1 1>
    R2 Crimson:LinearColor=<rgba 0.9478 0.0642 0.0881 1>
    R2 DefaultColor:LinearColor=<rgba 0.5841 1 0.2195 1>
    R2 ForestGreen:LinearColor=<rgba 0 0.3542 0.0179 1>
    R2 Orange:LinearColor=<rgba 0.849 0.3566 0.0929 1>
    R2 Purple:LinearColor=<rgba 0.2449 0.0564 0.5156 1>
    Saffron=<rgba 0.9 0.7294 0.1687 1>
    SkyBlue=<rgba 0.2406 0.8124 1.1 1>
    TitaniumWhite=<rgba 1 1 1 1>
    Use R1 Painted Overrides=true
    Use R2 Painted Overrides=true
   ) {module=VC_PaintedVariantsOverride001}
   /SetVariables(Particles.Gradient:LinearColor=&ColorFromCurve(ColorCurve=DataInterface() {propertyValues {
       AlphaCurve {keys [<at 0 1 RCIM_Cubic>,<at 1 0 RCIM_Cubic>]}
       BlueCurve {keys [<at 0.1007 0.4179 RCIM_Cubic>,<at 0.4241 0.285 RCIM_Cubic>]}
       GreenCurve {keys [<at 0.1007 0.6867 RCIM_Cubic>,<at 0.4241 0.0314 RCIM_Cubic>]}
       RedCurve {keys [<at 0.1007 0.0976 RCIM_Cubic>,<at 0.4241 0 RCIM_Cubic>]}
      }})) {enabled=false, isSetParameters=true}
  ]
  rendererClasses [&NiagaraRibbonRendererProperties]
  renderers [
   {rendererClass=&NiagaraRibbonRendererProperties}
   {rendererClass=&NiagaraRibbonRendererProperties, rendererIndex=1}
   {rendererClass=&NiagaraRibbonRendererProperties, rendererIndex=2}
  ]
 }
 NE_VC_Boost_Cone_Spawn {
  emitterUpdateScript [
   &EmitterState(Inactive Response=::NewEnumerator1, Life Cycle Mode=::NewEnumerator1, Loop Duration=1, UseLoopDelay=false)
   &VC_FilteredSocketCount()
   &SpawnRate(SpawnRate=5, Use Spawn Probability=false)
   &SpawnRate(SpawnGroup=1, SpawnRate=5, Use Spawn Probability=false) {module=SpawnRate002}
   &SpawnBurst_Instantaneous(Spawn Count=3, Spawn Group=2, Use Spawn Probability=false, UseLoopCountLimit=false)
   &SpawnBurst_Instantaneous(Spawn Count=3, Spawn Time=4.2, Use Spawn Probability=false, UseLoopCountLimit=false) {module=SpawnBurst_Instantaneous001}
  ]
  particleSpawnScript [
   &VC_PaintedVariantsOverride(
    Black=<rgba 0.022 0.022 0.022 1>
    BurntSienna=<rgba 0.1719 0.0637 0 1>
    Cobalt=<rgba 0 0.3806 0.8385 1>
    Crimson=<rgba 1 0.1198 0.1436 1>
    DefaultColor=<rgba 0.0425 1 0.1565 1>
    ForestGreen=<rgba 0.0406 0.5 0.0156 1>
    Grey=<rgba 0.1458 0.1458 0.1458 1>
    Orange=<rgba 0.6724 0.294 0.0385 1>
    Pink=<rgba 1.5 0.0859 0.9858 1>
    Purple=<rgba 0.2519 0 0.8281 1>
    R1 Black:LinearColor=<rgba 0.3073 0.3073 0.3073 1>
    R1 BurntSienna:LinearColor=<rgba 0.8522 0.4008 0.2796 1>
    R1 Cobalt:LinearColor=<rgba 0.3734 0.7214 0.9688 1>
    R1 Crimson:LinearColor=<rgba 0.6979 0.4326 0.4398 1>
    R1 DefaultColor=<rgba 1 1 1 0.8>
    R1 ForestGreen:LinearColor=<rgba 0.3129 0.8177 0.2939 1>
    R1 Grey:LinearColor=<rgba 0.4219 0.4219 0.4219 1>
    R1 Lime:LinearColor=<rgba 0.5716 0.8958 0.2473 1>
    R1 Orange:LinearColor=<rgba 0.9792 0.9019 0.14 1>
    R1 Pink:LinearColor=<rgba 1.421 0.3359 1.5 1>
    R1 Purple:LinearColor=<rgba 0.4754 0.2407 0.9167 1>
    R1 Saffron:LinearColor=<rgba 0.9 0.7469 0.2437 1>
    R1 SkyBlue:LinearColor=<rgba 0.4182 0.7024 0.8542 1>
    R2 Black:LinearColor=<rgba 0.01 0.01 0.01 0.65>
    R2 BurntSienna:LinearColor=<rgba 0.0642 0.027 0.0171 0.65>
    R2 Cobalt:LinearColor=<rgba 0 0.0145 0.7865 0.65>
    R2 Crimson:LinearColor=<rgba 0.4896 0.0178 0.0306 0.65>
    R2 DefaultColor:LinearColor=<rgba 0.0976 0.6867 0.4179 0.65>
    R2 ForestGreen:LinearColor=<rgba 0 0.2031 0.0006 0.65>
    R2 Grey:LinearColor=<rgba 0.1354 0.1354 0.1354 0.65>
    R2 Lime:LinearColor=<rgba 0.2742 0.5885 0.049 0.65>
    R2 Orange:LinearColor=<rgba 0.9844 0.0595 0 0.65>
    R2 Pink:LinearColor=<rgba 0.5052 0.0632 0.365 0.65>
    R2 Purple:LinearColor=<rgba 0.0514 0 0.6795 0.65>
    R2 Saffron:LinearColor=<rgba 0.8632 0.8879 0.0762 0.65>
    R2 SkyBlue:LinearColor=<rgba 0 0.357 0.5365 0.65>
    R2 TitaniumWhite:LinearColor=<rgba 0.5781 0.5781 0.5781 0.65>
    Saffron=<rgba 1 0.7691 0.0104 1>
    SkyBlue=<rgba 0.0573 0.6846 1 1>
    Use R1 Painted Overrides=true
    Use R2 Painted Overrides=true
   )
   /SetVariables(
    Particles.AquaSize:Vector2f=&Multiply_Vector2D_ByFloat(Float=45, Vector2D=<1 1>)
    Particles.VisibilityTag=&SpawnGroup()
   ) {isSetParameters=true, module=SetVariables_#2}
   /VehicleCosmetics/FX/Niagara/Modules/MergeEmitters/VC_ExtraBindingColors(
    ExtraColors=1
    R1 Color:LinearColor=&MakeCustomLinearColorFromBool(Bool=$User.bIsPainted, True Color=$User.PaintedColor, False Color=<rgba 0 0 0 1>)
   ) {enabled=false}
   &InitializeParticle(
    Color Mode=::NewEnumerator0
    Lifetime=&SelectFloatFromArray(
     Direct Array Index=&SpawnGroup()
     Float Selection Array=DataInterface() {propertyValues {FloatData [0.6,0.6,1]}}
    )
    Mesh Scale=<6.5 6.5 14>
    Mesh Scale Mode=::NewEnumerator4
    Sprite Rotation Mode=::NewEnumerator2
    Sprite Size Mode=::NewEnumerator3
    Uniform Sprite Size=28.5
    UsePositionOffset=false
    Write Material Random=true
   )
   &PartitionParticles(Number Of Partitions=$Emitter.PortCount)
   &SkeletalMeshLocation(
    Bone / Socket Index=$Particles.PartitionParticles.Partition
    Bone / Socket Sampling Mode=::NewEnumerator5
    Interpolate Mesh Sampling=false
    MaskSpawnGroup=false
    Mesh Sampling Type=::NewEnumerator6
    Orientation Sampling=::NewEnumerator1
    Sampled Position Offset=&AddVectorToPosition(
     Position=<-70 0 0>
     Vector=&AddVectorToPosition(
      Position=&SelectVectorFromArray(
       Direct Array Index=&SpawnGroup()
       Direct Array Mode=::NewEnumerator1
       Vector Selection Array=DataInterface() {propertyValues {FloatData [<10 0 0>, <10 0 0>, <1 0 0>]}}
      )
      Vector=$User.DTCarOffset
     )
    )
    Sampled Position Offset Coordinate Space:ENiagaraSkelMeshTransforms=::NewEnumerator2
    Skeletal Mesh=$System.SkeletalMesh
   )
   &AddVelocity(Velocity=&SelectVectorFromArray(
     Direct Array Index=&SpawnGroup()
     Vector Selection Array=DataInterface() {propertyValues {FloatData [<-90 0 0>, <-90 1 1>, <0 0 0>]}}
    ))
  ]
  particleUpdateScript [
   &ParticleState()
   /SetVariables(Particles.Rings_M3_Dynamic:Vector4f=&MakeVector4(
     X=&MakeCustomFloatFromBool(Bool=$User.bIsPainted, True Float=$User.RingsOpacity, False Float=1)
     Y=0
     Z=0
     W=0
    )) {isSetParameters=true, module=SetVariables_#3}
   &ScaleMeshSize(Scale Factor=&VectorFromFloat(Value=&FloatFromCurve(FloatCurve=DataInterface() {propertyValues {Curve {
         defaultValue=1
         keys [{interpMode=RCIM_Cubic, tangentMode=RCTM_User, time=0, value=1}, {interpMode=RCIM_Cubic, tangentMode=RCTM_User, time=1, value=0.8}]
        }}}))) {module=ScaleMeshSize001}
   &ScaleColor(
    Scale Alpha=&MaskFloatBySpawnGroup(
     Masked Float=&FloatFromCurve(FloatCurve=DataInterface() {propertyValues {Curve {
         defaultValue=1
         keys [
          {arriveTangent=9.0725, interpMode=RCIM_Cubic, leaveTangent=9.0725, tangentMode=RCTM_User, time=0, value=0}
          {arriveTangent=-2.5249, interpMode=RCIM_Cubic, leaveTangent=-2.5249, tangentMode=RCTM_User, time=0.1408, value=0.9847}
          {arriveTangent=-0.0405, interpMode=RCIM_Cubic, leaveTangent=-0.0405, tangentMode=RCTM_User, time=1, value=0}
         ]
        }}})
     Passthrough Float=&MaskFloatBySpawnGroup(
      Masked Float=&FloatFromCurve(FloatCurve=DataInterface() {propertyValues {Curve {
          defaultValue=1
          keys [
           {arriveTangent=9.0725, interpMode=RCIM_Cubic, leaveTangent=9.0725, tangentMode=RCTM_User, time=0, value=0}
           {arriveTangent=-2.5249, interpMode=RCIM_Cubic, leaveTangent=-2.5249, tangentMode=RCTM_User, time=0.1408, value=0.9847}
           {arriveTangent=-0.0405, interpMode=RCIM_Cubic, leaveTangent=-0.0405, tangentMode=RCTM_User, time=1, value=0}
          ]
         }}})
      Passthrough Float=1
      Spawn Group Mask=1
     )
    )
    ScaleA=true
    ScaleRGB=true
   )
   &SolveForcesAndVelocity(Clamp Velocity=false, Limit Acceleration=false)
  ]
  rendererClasses [&NiagaraMeshRendererProperties, &NiagaraSpriteRendererProperties]
  renderers [
   {rendererClass=&NiagaraMeshRendererProperties, rendererIndex=0}
   {rendererClass=&NiagaraMeshRendererProperties, rendererIndex=1}
   {rendererClass=&NiagaraSpriteRendererProperties, rendererIndex=2}
   {rendererClass=&NiagaraSpriteRendererProperties, rendererIndex=3}
  ]
 }
 NE_VC_Boost_Cone_Spawn001 {
  bEnabled=false
  emitterUpdateScript [
   &EmitterState(Inactive Response=::NewEnumerator1, Life Cycle Mode=::NewEnumerator1, Loop Duration=4.2, UseLoopDelay=false)
   &VC_FilteredSocketCount()
   &SpawnRate(SpawnRate=5, Use Spawn Probability=false)
   &SpawnRate(SpawnGroup=1, SpawnRate=5, Use Spawn Probability=false) {module=SpawnRate002}
   &SpawnBurst_Instantaneous(Spawn Count=3, Spawn Group=2, Use Spawn Probability=false, UseLoopCountLimit=false)
   &SpawnBurst_Instantaneous(Spawn Count=3, Spawn Time=4.2, Use Spawn Probability=false, UseLoopCountLimit=false) {module=SpawnBurst_Instantaneous001}
  ]
  particleSpawnScript [
   &VC_PaintedVariantsOverride(
    Black=<rgba 0.022 0.022 0.022 1>
    BurntSienna=<rgba 0.1719 0.0637 0 1>
    Cobalt=<rgba 0 0.3806 0.8385 1>
    Crimson=<rgba 1 0.1198 0.1436 1>
    DefaultColor=<rgba 0.0425 1 0.1565 1>
    ForestGreen=<rgba 0.0406 0.5 0.0156 1>
    Grey=<rgba 0.1458 0.1458 0.1458 1>
    Orange=<rgba 0.6724 0.294 0.0385 1>
    Pink=<rgba 1.5 0.0859 0.9858 1>
    Purple=<rgba 0.2519 0 0.8281 1>
    R1 Black:LinearColor=<rgba 0.3073 0.3073 0.3073 1>
    R1 BurntSienna:LinearColor=<rgba 0.8522 0.4008 0.2796 1>
    R1 Cobalt:LinearColor=<rgba 0.3734 0.7214 0.9688 1>
    R1 Crimson:LinearColor=<rgba 0.6979 0.4326 0.4398 1>
    R1 DefaultColor=<rgba 1 1 1 0.8>
    R1 ForestGreen:LinearColor=<rgba 0.3129 0.8177 0.2939 1>
    R1 Grey:LinearColor=<rgba 0.4219 0.4219 0.4219 1>
    R1 Lime:LinearColor=<rgba 0.5716 0.8958 0.2473 1>
    R1 Orange:LinearColor=<rgba 0.9792 0.9019 0.14 1>
    R1 Pink:LinearColor=<rgba 1.421 0.3359 1.5 1>
    R1 Purple:LinearColor=<rgba 0.4754 0.2407 0.9167 1>
    R1 Saffron:LinearColor=<rgba 0.9 0.7469 0.2437 1>
    R1 SkyBlue:LinearColor=<rgba 0.4182 0.7024 0.8542 1>
    R2 Black:LinearColor=<rgba 0.01 0.01 0.01 0.65>
    R2 BurntSienna:LinearColor=<rgba 0.0642 0.027 0.0171 0.65>
    R2 Cobalt:LinearColor=<rgba 0 0.0145 0.7865 0.65>
    R2 Crimson:LinearColor=<rgba 0.4896 0.0178 0.0306 0.65>
    R2 DefaultColor:LinearColor=<rgba 0.0976 0.6867 0.4179 0.65>
    R2 ForestGreen:LinearColor=<rgba 0 0.2031 0.0006 0.65>
    R2 Grey:LinearColor=<rgba 0.1354 0.1354 0.1354 0.65>
    R2 Lime:LinearColor=<rgba 0.2742 0.5885 0.049 0.65>
    R2 Orange:LinearColor=<rgba 0.9844 0.0595 0 0.65>
    R2 Pink:LinearColor=<rgba 0.5052 0.0632 0.365 0.65>
    R2 Purple:LinearColor=<rgba 0.0514 0 0.6795 0.65>
    R2 Saffron:LinearColor=<rgba 0.8632 0.8879 0.0762 0.65>
    R2 SkyBlue:LinearColor=<rgba 0 0.357 0.5365 0.65>
    R2 TitaniumWhite:LinearColor=<rgba 0.5781 0.5781 0.5781 0.65>
    Saffron=<rgba 1 0.7691 0.0104 1>
    SkyBlue=<rgba 0.0573 0.6846 1 1>
    Use R1 Painted Overrides=true
    Use R2 Painted Overrides=true
   )
   /SetVariables(
    Particles.AquaSize:Vector2f=&Multiply_Vector2D_ByFloat(Float=45, Vector2D=<1 1>)
    Particles.VisibilityTag=&SpawnGroup()
   ) {isSetParameters=true, module=SetVariables_#2}
   &VC_ExtraBindingColors(
    ExtraColors=1
    R1 Color:LinearColor=&MakeCustomLinearColorFromBool(Bool=$User.bIsPainted, True Color=$User.PaintedColor, False Color=<rgba 0 0 0 1>)
   ) {enabled=false}
   &InitializeParticle(
    Color Mode=::NewEnumerator0
    Lifetime=&SelectFloatFromArray(
     Direct Array Index=&SpawnGroup()
     Float Selection Array=DataInterface() {propertyValues {FloatData [0.6,0.6,1]}}
    )
    Mesh Scale=<6.5 6.5 14>
    Mesh Scale Mode=::NewEnumerator4
    Sprite Rotation Mode=::NewEnumerator2
    Sprite Size Mode=::NewEnumerator3
    Uniform Sprite Size=28.5
    UsePositionOffset=false
    Write Material Random=true
   )
   &PartitionParticles(Number Of Partitions=$Emitter.PortCount)
   &SkeletalMeshLocation(
    Bone / Socket Index=$Particles.PartitionParticles.Partition
    Bone / Socket Sampling Mode=::NewEnumerator5
    Interpolate Mesh Sampling=false
    MaskSpawnGroup=false
    Mesh Sampling Type=::NewEnumerator6
    Orientation Sampling=::NewEnumerator1
    Sampled Position Offset=&AddVectorToPosition(
     Position=<-70 0 0>
     Vector=&AddVectorToPosition(
      Position=&SelectVectorFromArray(
       Direct Array Index=&SpawnGroup()
       Direct Array Mode=::NewEnumerator1
       Vector Selection Array=DataInterface() {propertyValues {FloatData [<10 0 0>, <10 0 0>, <1 0 0>]}}
      )
      Vector=$User.DTCarOffset
     )
    )
    Sampled Position Offset Coordinate Space:ENiagaraSkelMeshTransforms=::NewEnumerator2
    Skeletal Mesh=$System.SkeletalMesh
   )
   &AddVelocity(Velocity=&SelectVectorFromArray(
     Direct Array Index=&SpawnGroup()
     Vector Selection Array=DataInterface() {propertyValues {FloatData [<-90 0 0>, <-90 1 1>, <0 0 0>]}}
    ))
  ]
  particleUpdateScript [
   &ParticleState()
   /SetVariables(Particles.Rings_M3_Dynamic:Vector4f=&MakeVector4(
     X=&MakeCustomFloatFromBool(Bool=$User.bIsPainted, True Float=$User.RingsOpacity, False Float=1)
     Y=0
     Z=0
     W=0
    )) {isSetParameters=true, module=SetVariables_#3}
   &ScaleMeshSize(Scale Factor=&VectorFromFloat(Value=&FloatFromCurve(FloatCurve=DataInterface() {propertyValues {Curve {
         defaultValue=1
         keys [{interpMode=RCIM_Cubic, tangentMode=RCTM_User, time=0, value=1}, {interpMode=RCIM_Cubic, tangentMode=RCTM_User, time=1, value=0.8}]
        }}}))) {module=ScaleMeshSize001}
   &ScaleColor(
    Scale Alpha=&MaskFloatBySpawnGroup(
     Masked Float=&FloatFromCurve(FloatCurve=DataInterface() {propertyValues {Curve {
         defaultValue=1
         keys [
          {arriveTangent=9.0725, interpMode=RCIM_Cubic, leaveTangent=9.0725, tangentMode=RCTM_User, time=0, value=0}
          {arriveTangent=-2.5249, interpMode=RCIM_Cubic, leaveTangent=-2.5249, tangentMode=RCTM_User, time=0.1408, value=0.9847}
          {arriveTangent=-0.0405, interpMode=RCIM_Cubic, leaveTangent=-0.0405, tangentMode=RCTM_User, time=1, value=0}
         ]
        }}})
     Passthrough Float=&MaskFloatBySpawnGroup(
      Masked Float=&FloatFromCurve(FloatCurve=DataInterface() {propertyValues {Curve {
          defaultValue=1
          keys [
           {arriveTangent=9.0725, interpMode=RCIM_Cubic, leaveTangent=9.0725, tangentMode=RCTM_User, time=0, value=0}
           {arriveTangent=-2.5249, interpMode=RCIM_Cubic, leaveTangent=-2.5249, tangentMode=RCTM_User, time=0.1408, value=0.9847}
           {arriveTangent=-0.0405, interpMode=RCIM_Cubic, leaveTangent=-0.0405, tangentMode=RCTM_User, time=1, value=0}
          ]
         }}})
      Passthrough Float=1
      Spawn Group Mask=1
     )
    )
    ScaleA=true
    ScaleRGB=true
   )
   &SolveForcesAndVelocity(Clamp Velocity=false, Limit Acceleration=false)
  ]
  rendererClasses [&NiagaraMeshRendererProperties, &NiagaraSpriteRendererProperties]
  renderers [
   {rendererClass=&NiagaraMeshRendererProperties, rendererIndex=0}
   {rendererClass=&NiagaraMeshRendererProperties, rendererIndex=1}
   {rendererClass=&NiagaraSpriteRendererProperties, rendererIndex=2}
   {rendererClass=&NiagaraSpriteRendererProperties, rendererIndex=3}
  ]
 }
 NE_VC_Boost_Sprites {
  emitterUpdateScript [
   &EmitterState(Inactive Response=::NewEnumerator1, Life Cycle Mode=::NewEnumerator1, Loop Duration=1, UseLoopDelay=false)
   &VC_FilteredSocketCount()
   &SpawnRate(SpawnRate=5, Use Spawn Probability=false)
   &SpawnRate(SpawnGroup=1, SpawnRate=5, Use Spawn Probability=false) {module=SpawnRate002}
   &SpawnRate(SpawnGroup=2, SpawnRate=24, Use Spawn Probability=false) {module=SpawnRate003}
  ]
  particleSpawnScript [
   &VC_PaintedVariantsOverride(
    Black=<rgba 0.0625 0.0625 0.0625 1>
    BurntSienna=<rgba 0.18 0.058 0.0253 2>
    Cobalt=<rgba 0 0.1415 0.9219 2>
    Crimson=<rgba 0.799 0.0479 0.0683 2.5>
    DefaultColor=<rgba 0.6007 1.3 0.1016 1>
    ForestGreen=<rgba 0.1108 0.8646 0 1>
    Grey=<rgba 0.151 0.151 0.151 1.5>
    Lime=<rgba 0.4235 0.799 0.0479 1>
    Orange=<rgba 0.8542 0.3066 0.0133 1>
    Pink=<rgba 0.9375 0.1953 0.6676 1>
    Purple=<rgba 0.1839 0.0654 0.9652 1.2>
    R2 Black:LinearColor=&SelectLinearColorFromArray(
     Array Sampling Mode=::NewEnumerator0
     Direct Array Index=?
     Color Selection Array=DataInterface() {propertyValues {ColorData [<rgba 0.3333 0.3333 0.3333 1>, <rgba 0.125 0.125 0.125 1>]}}
     Random Seed=0
    )
    R2 BurntSienna:LinearColor=&SelectLinearColorFromArray(
     Array Sampling Mode=::NewEnumerator0
     Direct Array Index=?
     Color Selection Array=DataInterface() {propertyValues {ColorData [<rgba 0.3333 0.117 0 1>, <rgba 0.0885 0.0265 0.0101 1>]}}
     Random Seed=0
    )
    R2 Cobalt:LinearColor=&SelectLinearColorFromArray(
     Array Sampling Mode=::NewEnumerator0
     Direct Array Index=?
     Color Selection Array=DataInterface() {propertyValues {ColorData [<rgba 0 0.8367 1 1>, <rgba 0 0.388 0.6957 1>]}}
     Random Seed=0
    )
    R2 Crimson:LinearColor=&SelectLinearColorFromArray(
     Array Sampling Mode=::NewEnumerator0
     Direct Array Index=?
     Color Selection Array=DataInterface() {propertyValues {ColorData [<rgba 1 0.2445 0.1771 1>, <rgba 1 0.5099 0.4844 1>]}}
     Random Seed=0
    )
    R2 DefaultColor:LinearColor=&SelectLinearColorFromArray(
     Array Sampling Mode=::NewEnumerator0
     Direct Array Index=?
     Color Selection Array=DataInterface() {propertyValues {ColorData [<rgba 1.1552 2.5 0.1955 1>, <rgba 0.1173 1.5 1.5 1>]}}
     Random Seed=0
    )
    R2 ForestGreen:LinearColor=&SelectLinearColorFromArray(
     Array Sampling Mode=::NewEnumerator0
     Direct Array Index=?
     Color Selection Array=DataInterface() {propertyValues {ColorData [<rgba 0.0339 0.9652 0.0302 1>, <rgba 0.3124 1 0.2212 1>]}}
     Random Seed=0
    )
    R2 Grey:LinearColor=&SelectLinearColorFromArray(
     Array Sampling Mode=::NewEnumerator0
     Direct Array Index=?
     Color Selection Array=DataInterface() {propertyValues {ColorData [<rgba 0.2604 0.2604 0.2604 1>, <rgba 0.125 0.125 0.125 1>]}}
     Random Seed=0
    )
    R2 Lime:LinearColor=&SelectLinearColorFromArray(
     Array Sampling Mode=::NewEnumerator0
     Direct Array Index=?
     Color Selection Array=DataInterface() {propertyValues {ColorData [<rgba 1.8254 3 0.0781 1>, <rgba 0.8803 2 0.0938 1>]}}
     Random Seed=0
    )
    R2 Orange:LinearColor=&SelectLinearColorFromArray(
     Array Sampling Mode=::NewEnumerator0
     Direct Array Index=?
     Color Selection Array=DataInterface() {propertyValues {ColorData [<rgba 3 2.0921 0.3342 1>, <rgba 2 0.9309 0.0783 1>]}}
     Random Seed=0
    )
    R2 Pink:LinearColor=&SelectLinearColorFromArray(
     Array Sampling Mode=::NewEnumerator0
     Direct Array Index=?
     Color Selection Array=DataInterface() {propertyValues {ColorData [<rgba 3 0.2718 2.4659 1>, <rgba 2 0.2721 1.6288 1>]}}
     Random Seed=0
    )
    R2 Purple:LinearColor=&SelectLinearColorFromArray(
     Array Sampling Mode=::NewEnumerator0
     Direct Array Index=?
     Color Selection Array=DataInterface() {propertyValues {ColorData [<rgba 0.2812 0 3 1>, <rgba 0.1489 0 1.5 1>]}}
     Random Seed=0
    )
    R2 Saffron:LinearColor=&SelectLinearColorFromArray(
     Array Sampling Mode=::NewEnumerator0
     Direct Array Index=?
     Color Selection Array=DataInterface() {propertyValues {ColorData [<rgba 1 0.9413 0.3854 1>, <rgba 1 0.8807 0.3079 1>]}}
     Random Seed=0
    )
    R2 SkyBlue:LinearColor=&SelectLinearColorFromArray(
     Array Sampling Mode=::NewEnumerator0
     Direct Array Index=?
     Color Selection Array=DataInterface() {propertyValues {ColorData [<rgba 0.1562 0.9022 1 1>, <rgba 0.1979 1.3598 2 1>]}}
     Random Seed=0
    )
    Saffron=<rgba 0.8698 0.6996 0.1404 1.2>
    SkyBlue=<rgba 0.1243 0.7035 0.9948 1>
    TitaniumWhite=<rgba 1 1 1 1>
    Use R2 Painted Overrides=true
   )
   /SetVariables(Particles.VisibilityTag=&SpawnGroup()) {isSetParameters=true}
   &InitializeParticle(
    Color Mode=::NewEnumerator0
    Lifetime=&MergeHelper_FloatSwitch(
     0=&RandomRangeFloat(Minimum=0.5, Maximum=1)
     1=&RandomRangeFloat(Minimum=0.5, Maximum=1)
     2=&RandomRangeFloat(Minimum=0.2, Maximum=0.4)
     3=?
     Index=$System.GameModeID
     Num=::NewEnumerator1
    )
    Sprite Rotation Mode=::NewEnumerator2
    Sprite Size Mode=::NewEnumerator3
    Uniform Sprite Size=&MaskFloatBySpawnGroup(
     Masked Float=&RandomRangeFloat(Minimum=20, Maximum=40)
     Passthrough Float=&MaskFloatBySpawnGroup(
      Masked Float=&RandomRangeFloat(Minimum=20, Maximum=30)
      Passthrough Float=&RandomRangeFloat(Minimum=5, Maximum=12)
      Spawn Group Mask=1
     )
    )
    UsePositionOffset=false
    Write Material Random=true
   )
   &PartitionParticles(Number Of Partitions=$Emitter.PortCount, Particles Per Partition=2, Partition Mode:ENiagara_PartitionMode=::NewEnumerator1)
   &SkeletalMeshLocation(
    Bone / Socket Index=$Particles.PartitionParticles.Partition
    Bone / Socket Sampling Mode=::NewEnumerator5
    MaskSpawnGroup=false
    Mesh Sampling Type=::NewEnumerator6
    Sampled Position Offset=&AddVectorToPosition(Position=<-100 0 0>, Vector=$User.DTCarOffset)
    Skeletal Mesh=$System.SkeletalMesh
   )
   &ShapeLocation(Hemisphere Distribution=<0.5 1>, MaskSpawnGroup=false, Override Seed=false, Spawn Group Mask=1, Sphere Radius=30, UseOwnerScale=true)
   &AddVelocity(Velocity=&RandomRangeVector(
     Minimum=&MergeHelper_VectorSwitch(
      0=&SelectVectorFromArray(
       Direct Array Index=&SpawnGroup()
       Vector Selection Array=DataInterface() {propertyValues {FloatData [<-1400 -50 -50>, <-1400 -50 -50>, <-1400 -67.5 -67.5>]}}
      )
      1=&SelectVectorFromArray(
       Direct Array Index=&SpawnGroup()
       Vector Selection Array=DataInterface() {propertyValues {FloatData [<-500 -50 -50>, <-500 -50 -50>, <-500 -67.5 -67.5>]}}
      )
      2=&SelectVectorFromArray(
       Direct Array Index=&SpawnGroup()
       Vector Selection Array=DataInterface() {propertyValues {FloatData [<0 -50 -5>, <0 -50 -5>, <0 -67.5 -6.75>]}}
      )
      3=?
      Index=$System.GameModeID
      Num=::NewEnumerator1
     )
     Maximum=&MergeHelper_VectorSwitch(
      0=&SelectVectorFromArray(
       Direct Array Index=&SpawnGroup()
       Vector Selection Array=DataInterface() {propertyValues {FloatData [<-50 50 50>, <-50 50 50>, <-50 67.5 67.5>]}}
      )
      1=&SelectVectorFromArray(
       Direct Array Index=&SpawnGroup()
       Vector Selection Array=DataInterface() {propertyValues {FloatData [<-50 50 50>, <-50 50 50>, <-50 67.5 67.5>]}}
      )
      2=&SelectVectorFromArray(
       Direct Array Index=&SpawnGroup()
       Vector Selection Array=DataInterface() {propertyValues {FloatData [<50 50 50>, <50 50 50>, <50 67.5 67.5>]}}
      )
      3=?
      Index=$System.GameModeID
      Num=::NewEnumerator1
     )
    ))
  ]
  particleUpdateScript [
   &ParticleState()
   &SpriteRotationRate(Rotation Rate=&RandomRangeFloat(Minimum=-800, Maximum=800))
   &SubUVAnimation(
    Play Rate:NiagaraFloat=&RandomRangeFloat(Minimum=0.5, Maximum=2)
    Random Start Frame=true
    "Sprite Renderer "=DataInterface() {propertyValues {SpriteRenderer=&self:NE_VC_Boost_Sprites_0.NiagaraSpriteRendererProperties_4}}
    SubUV Animation Mode:ENiagara_SubUVLookupModeV2=::NewEnumerator3
    UseEndFrame=false
    UsePlayRateScale=false
    UseStartFrame=false
   )
   &ScaleColor(
    Scale Alpha=&FloatFromCurve(FloatCurve=DataInterface() {propertyValues {Curve {
        defaultValue=1
        keys [{interpMode=RCIM_Cubic, tangentMode=RCTM_User, time=0.6, value=1}, {interpMode=RCIM_Cubic, tangentMode=RCTM_User, time=1, value=0}]
       }}})
    ScaleA=true
    ScaleRGB=true
   )
   &SolveForcesAndVelocity(Clamp Velocity=false, Limit Acceleration=false)
  ]
  rendererClasses [&NiagaraSpriteRendererProperties]
  renderers [
   {rendererClass=&NiagaraSpriteRendererProperties}
   {rendererClass=&NiagaraSpriteRendererProperties, rendererIndex=1}
   {rendererClass=&NiagaraSpriteRendererProperties, rendererIndex=2}
  ]
 }
}
summary {
 systemName=NS_Boost_Lifeline
 emitters [
  {emitterName=NE_VC_Boost_Sprites, rendererClasses [&NiagaraSpriteRendererProperties]}
  {
   emitterName=NE_VC_Boost_Cone_Spawn
   rendererClasses [&NiagaraMeshRendererProperties, &NiagaraSpriteRendererProperties]
  }
  {emitterName=NE_LifelineRibbons, rendererClasses [&NiagaraRibbonRendererProperties]}
  {
   bEnabled=false
   emitterName=NE_VC_Boost_Cone_Spawn001
   rendererClasses [&NiagaraMeshRendererProperties, &NiagaraSpriteRendererProperties]
  }
 ]
 userVariables [
  User.bIsFrontendPreview {type=NiagaraBool, defaultValue=true}
  User.bIsPainted {type=NiagaraBool}
  User.ColorVariantID {type=NiagaraInt32}
  User.ColorWhite {type=NiagaraFloat, defaultValue=0, description=""}
  User.DTCarOffset {type=NiagaraPosition, description="", defaultValue {}}
  User.DTCarOffset {type=Vector3f, description="", defaultValue {struct=Vector3f, value=<0 0 0>}}
  User.Emissive {type=NiagaraFloat, defaultValue=10, description=""}
  User.EmissiveTrail {type=NiagaraFloat, defaultValue=1, description=""}
  User.EmissiveTrailAlpha {type=NiagaraFloat, defaultValue=0.6, description=""}
  User.GameModeID {type=NiagaraInt32}
  User.InColorIntensity {type=NiagaraFloat, defaultValue=1, description=""}
  User.OutColorIntensity {type=NiagaraFloat, defaultValue=1, description=""}
  User.PaintedColor {type=LinearColor, description="", defaultValue {struct=LinearColor, value=<rgba 1 1 1 1>}}
  User.PaintedScalar {type=NiagaraFloat, defaultValue=1, description=""}
  User.PortCount {type=NiagaraInt32, defaultValue=2}
  User.RingsIntensityB {type=NiagaraFloat, defaultValue=5, description=""}
  User.RingsIntensityG {type=NiagaraFloat, defaultValue=5, description=""}
  User.RingsIntensityR {type=NiagaraFloat, defaultValue=5, description=""}
  User.RingsOpacity {type=NiagaraFloat, defaultValue=1, description=""}
  User.Shape01 {type=LinearColor, description="", defaultValue {struct=LinearColor, value=<rgba 0 0.1275 1 1>}}
  User.Shape02 {type=LinearColor, description="", defaultValue {struct=LinearColor, value=<rgba 1 0 0.0097 1>}}
  User.SingleColorIntensity {type=NiagaraFloat, defaultValue=1, description=""}
 ]
}
